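# Niagara deploy effect for the Donut glider, a front-end variant going by its name, in which only the Light emitter is enabled.
# It bursts one short-lived light that follows the skeletal mesh's FX_Horn socket, cycling through a boosted rainbow gradient as alpha rises and falls.
# Confetti, ribbon contrail, sprite and spark emitters remain but are disabled, leaving the User.Sparks and User.SkeletalMesh inputs without visible effect.
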NiagaraSystem /BRCosmetics/Gadgets/Assets/VinderTech_GliderChute/Glider_Donut/FX/NS_Glider_Donut_Deploy_FrontEnd
dependencies {
 usedDataInterfaces [&NiagaraDataInterfaceCurve, &NiagaraDataInterfaceColorCurve, &NiagaraDataInterfaceVectorCurve]
 usedDynamicInputs []
 usedModules [&SystemState]
}
emitters {
 Confetti {
  bEnabled=false
  emitterUpdateScript [
   &EmitterState(Inactive Response=::NewEnumerator1, Life Cycle Mode=::NewEnumerator1, Loop Behavior=::NewEnumerator1, Loop Duration=3, UseLoopDelay=false)
   &SpawnRate(
    SpawnRate=&FloatFromCurve(
     FloatCurve=DataInterface() {propertyValues {Curve {keys [<at -0.0046 1.0113>,<at 1.0011 0.002>]}}}
     CurveIndex=$Emitter.NormalizedLoopAge
     Scale Curve=150
    )
    Use Spawn Probability=true
   )
  ]
  particleSpawnScript [
   /Niagara/Modules/Spawn/Initialization/InitializeParticle(
    Color=&LerpLinearColors(StartColor=<rgba 1 1 1 1>, EndColor=<rgba 1 1 1 1>, LerpFactor=1)
    Lifetime=&Multiply_Float(A=&UniformRangedFloat(Minimum=0.5, Maximum=1), B=1)
    Module.Write Color=true
    Module.Write Lifetime=true
    Module.Write Mass=true
    Module.Write Material Random=true
    Module.Write Position=false
    Module.Write Scale=false
    Module.Write SpriteRotation=true
    Module.Write SpriteSize=true
    Position=&SimulationPosition()
    Sprite Rotation=&UniformRangedFloat(Minimum=0, Maximum=360)
    Sprite Size=&UniformRangedVector2D(
     Minimum=&Vector2DFromFloat(Value=25)
     Maximum=&Vector2DFromFloat(Value=35)
    )
   )
   &SampleSkeletalMeshSkeleton(Mesh:NiagaraDataInterfaceSkeletalMesh=$User.SkeletalMesh)
   /Niagara/Modules/Spawn/Location/SkeletalMeshSkeletonLocation(Inherit Velocity From Mesh=150, Module.Inherit Velocity=false, Module.Use Sampled Bone as Ribbon ID=true)
   /Niagara/Modules/Spawn/Location/SphereLocation(
    MaskSpawnGroup=false
    Module.Override Randomness=false
    Sphere Radius=&UniformRangedFloat(Minimum=15, Maximum=30)
    Surface Only=false
   )
   /Niagara/Modules/Update/Forces/CurlNoiseForce(
    Noise Frequency=15
    Noise Strength=&Multiply_Float(A=0.15, B=&MakeFloatFromVector(Vector=$Engine.Owner.Velocity, Channel=::NewEnumerator0))
    Pan Noise=true
    Pan Noise Field=<0 10 0>
   ) {module=CurlNoiseForce001}
   &ApplyInitialForces()
  ]
  particleUpdateScript [
   &ParticleState()
   /SetVariables(Particles.GradientColor:Vector3f=&MakeVectorFromLinearColorRGB(Color=&MakeLinearColorFromVectorAndFloat(
      "Vector (RGB)"=&MakeVectorFromLinearColorRGB(Color=&ColorFromCurve(ColorCurve=DataInterface() {propertyValues {
          AlphaCurve {keys [<at 0.2533 1 RCIM_Cubic>,<at 0.9977 1 RCIM_Cubic>]}
          BlueCurve {keys [<at 0.4507 0 RCIM_Cubic>,<at 0.5102 0>,<at 0.6354 0>,<at 0.7224 1>,<at 0.7946 1.5>,<at 0.9219 1.5678 RCIM_Cubic>]}
          GreenCurve {keys [<at 0.4507 0.039 RCIM_Cubic>,<at 0.5102 0.8559>,<at 0.6354 1>,<at 0.7224 0.6207>,<at 0.7946 0.4532>,<at 0.9219 0 RCIM_Cubic>]}
          RedCurve {keys [<at 0.4507 1 RCIM_Cubic>,<at 0.5102 2>,<at 0.6354 0.3535>,<at 0.7224 0.0629>,<at 0.7946 0>,<at 0.9219 2 RCIM_Cubic>]}
         }}))
      "Float (Alpha)"=1
     ))) {isSetParameters=true}
   /SetVariables(Particles.Color=&MakeLinearColorFromVectorAndFloat(
     "Vector (RGB)"=&Lerp_Vector(
      A=&Multiply_VectorByFloat(Vector=<1 1 1>, Float=1)
      B=&Multiply_VectorByFloat(Vector=&UniformRangedVector(Minimum=$Particles.GradientColor, Maximum=$Particles.GradientColor), Float=1)
      Alpha=&FloatFromCurve(FloatCurve=DataInterface() {propertyValues {Curve {keys [
           {interpMode=RCIM_Cubic, tangentMode=RCTM_User, time=-0.0098, value=0.1132}
           {arriveTangent=0.0255, interpMode=RCIM_Cubic, leaveTangent=0.0255, tangentMode=RCTM_User, time=0.2458, value=0.1945}
           {arriveTangent=0.1228, interpMode=RCIM_Cubic, leaveTangent=0.1228, tangentMode=RCTM_User, time=0.5644, value=0.8025}
           {interpMode=RCIM_Cubic, tangentMode=RCTM_User, time=1.0011, value=0.002}
          ]}}})
     )
     "Float (Alpha)"=&FloatFromCurve(FloatCurve=DataInterface() {propertyValues {Curve {keys [
          {interpMode=RCIM_Cubic, tangentMode=RCTM_User, time=-0.0063, value=0.0184}
          {arriveTangent=1.3376, interpMode=RCIM_Cubic, leaveTangent=1.3376, tangentMode=RCTM_User, time=0.096, value=0.9808}
          {arriveTangent=0.3112, interpMode=RCIM_Cubic, leaveTangent=0.3112, tangentMode=RCTM_User, time=0.5407, value=1.3419}
          {arriveTangent=-3.7119, interpMode=RCIM_Cubic, leaveTangent=-3.7119, tangentMode=RCTM_User, time=0.7242, value=0.3749}
          {interpMode=RCIM_Cubic, tangentMode=RCTM_User, time=1.0011, value=0.002}
         ]}}})
    )) {isSetParameters=true, module=SetVariables_#1}
   &AddVelocity(Velocity=&RotateVectorByQuaternion(
     Vector To Rotate=&UniformRangedVector(Minimum=<-1 -1 -1>, Maximum=<1 1 1>)
     Quaternion=$Particles.SampleSkeletalMeshSkeleton.SampledRotation
    ))
   &AccelerationForce(Acceleration=<0 0 -800>)
   &CurlNoiseForce(
    Noise Frequency=1
    Noise Strength=&Multiply_Float(A=1, B=&MakeFloatFromVector(Vector=$Engine.Owner.Velocity, Channel=::NewEnumerator0))
    Pan Noise=true
    Pan Noise Field=<0 10 0>
   )
   &Drag(Drag=&UniformRangedFloat(Minimum=1, Maximum=2), Use Linear Drag=true, Use Rotational Drag=false)
   &InheritVelocity(
    Module.Limit Speed=true
    Velocity Limit=1500
    Velocity Scale=<-0.05 -0.05 -0.05>
    Velocity Source=&TransformVector(Vector=$Engine.Owner.Velocity, Source Space=::World, Destination Space=::Local)
   )
   &SolveForcesAndVelocity(Clamp Velocity=false, Limit Acceleration=false)
   &ScaleSpriteSize(Scale Factor=&Vector2DFromFloat(Value=&FloatFromCurve(FloatCurve=DataInterface() {propertyValues {Curve {keys [
          {arriveTangent=2.3096, interpMode=RCIM_Cubic, leaveTangent=2.3096, tangentMode=RCTM_User, time=0, value=0}
          {arriveTangent=-0.0952, interpMode=RCIM_Cubic, leaveTangent=-0.0952, tangentMode=RCTM_User, time=0.431, value=0.3973}
          <at 0.9969 0.0123 RCIM_Cubic>
         ]}}})))
   &ScaleSpriteSizeBySpeed(Max Scale Factor=<0.5 0.5>)
   &CameraOffset(Camera Offset Amount=25)
   /Niagara/Modules/Update/SubUV/SubUVAnimation(Number Of Frames=36)
  ]
  rendererClasses [&NiagaraSpriteRendererProperties]
  renderers [{rendererClass=&NiagaraSpriteRendererProperties}]
 }
 FN_Contrails_Ribbon_v4_A001 {
  bEnabled=false
  emitterUpdateScript [
   &EmitterScalability(
    Module.bApplySpawnCountScale=false
    Module.bApplySpawnCountScaleByDistanceFraction=false
    Module.bUseMaxDistance=false
    Module.bUseMinDistance=false
   ) {enabled=false}
   &EmitterLifeCycle(CompleteOnInactive=true) {enabled=false}
   &EmitterState()
   /Game/Effects/Niagara/Modules/Spawn/FN_SpawnContrails(All limbs=false, Global Spawnrate=10) {enabled=false}
   &FN_SpawnRibbon_v2(Spawn Mode=::NewEnumerator0, SpawnRate=25)
  ]
  particleSpawnScript [
   &FN_InitializeRibbon_v3(Custom Offset=true, Local Vector Offset:Vector3f=<150 0 0>, Ribbon Width=250)
   &AddVelocity(Velocity=<0 15 0>)
   /SetVariables(Particles.FadeOffAlpha:NiagaraFloat=&ClampFloat(Float=&Multiply_Float(A=$Emitter.Age, B=1))) {enabled=false, isSetParameters=true}
   &SystemLocation(MaskSpawnGroup=false, Offset=<0 -150 0>) {enabled=false}
  ]
  particleUpdateScript [
   &AccelerationForce(Acceleration=&VectorFromCurve(VectorCurve=DataInterface() {propertyValues {
       XCurve {keys [<at -0.0007 -157.5899>,<at 1.0152 -18.4419>]}
       YCurve {keys [<at -0.0011 1.0067>,<at 0.9988 0>]}
       ZCurve {keys [<at -0.0011 -300>,<at 0.9957 -200.05>]}
      }})) {enabled=false}
   &SampleSkeletalMeshSkeleton(Mesh=DataInterface() {propertyValues {
      DefaultMesh=null
      ExcludeBoneName=root
      OptionalComponentClass=null
      PreviewMesh=/BRCosmetics/Gadgets/Assets/VinderTech_GliderChute/Glider_Donut/Meshes/Donut
      SkinningMode=SkinOnTheFly
      SoftSourceActor=null
      SourceMode=Default
      UvSetIndex=0
      WholeMeshLOD=-1
      bExcludeBone=false
      bReadDeformedGeometry=true
      bRequireCurrentFrameData=true
      ComponentTags []
      FilteredBones [tail_1]
      FilteredSockets []
      MeshUserParameter {parameter {name=None, typeDefHandle {registeredTypeIndex=1}, varData []}}
      SamplingRegions []
     }}) {enabled=false}
   &SkeletalMeshSkeletonLocation(Inherit Velocity From Mesh=0, Module.Inherit Velocity=false, Module.Use Sampled Bone as Ribbon ID=true) {enabled=false}
   &Drag(Use Linear Drag=true, Use Rotational Drag=false) {enabled=false}
   &AddVelocity(Scale Added Velocity=<1 1 10>, Velocity=<0 -25 1>) {module=AddVelocity001}
   /Game/Effects/Niagara/Modules/Skydiving/FN_DivingAttributes/FN_DivingAttributes() {enabled=false}
   &InheritVelocity(
    Module.Limit Speed=true
    Velocity Limit=500
    Velocity Scale=<0.01 0 0.01>
    Velocity Source=&TransformVector(Vector=$Engine.Owner.Velocity, Source Space=::World, Destination Space=::Local)
   )
   &SolveForcesAndVelocity(Clamp Velocity=false, Limit Acceleration=false)
   &FN_UpdateRibbon(Initial Color:LinearColor=$Particles.Initial.Color, Initial RibbonWidth:NiagaraFloat=$Particles.Initial.RibbonWidth)
   /SetVariables(Particles.Velocity=$Particles.CustomizePosition) {enabled=false, isSetParameters=true}
   /SetVariables(Particles.GradientColor:Vector3f=&MakeVectorFromLinearColorRGB(Color=&MakeLinearColorFromVectorAndFloat(
      "Vector (RGB)"=&MakeVectorFromLinearColorRGB(Color=&ColorFromCurve(ColorCurve=DataInterface() {propertyValues {
          AlphaCurve {keys [<at -0.0023 1 RCIM_Cubic>,<at 0.9977 1 RCIM_Cubic>]}
          BlueCurve {keys [<at 0.0942 0 RCIM_Cubic>,<at 0.2809 0>,<at 0.4507 0>,<at 0.5951 1>,<at 0.7691 1.5>,<at 0.9071 1.5678 RCIM_Cubic>]}
          GreenCurve {keys [<at 0.0942 0.039 RCIM_Cubic>,<at 0.2809 0.8559>,<at 0.4507 1>,<at 0.5951 0.6207>,<at 0.7691 0.4532>,<at 0.9071 0 RCIM_Cubic>]}
          RedCurve {keys [<at 0.0942 1 RCIM_Cubic>,<at 0.2809 2>,<at 0.4507 0.3535>,<at 0.5951 0.0629>,<at 0.7691 0>,<at 0.9071 2 RCIM_Cubic>]}
         }}))
      "Float (Alpha)"=1
     ))) {isSetParameters=true}
   /SetVariables(Particles.Color=&MakeLinearColorFromVectorAndFloat(
     "Vector (RGB)"=&Multiply_Vector(A=$Particles.GradientColor, B=<2 2 2>)
     "Float (Alpha)"=&FloatFromCurve(FloatCurve=DataInterface() {propertyValues {Curve {keys [<at 0.184 0.0034>,<at 0.2538 0.8706>,<at 0.4416 0.8706>,<at 1.0011 0.002>]}}})
    )) {isSetParameters=true}
   /SetVariables(Particles.Color=&MultiplyLinearColor(
     Linear Color=$Particles.Color
     Scale Factor=&MakeFloat4FromVectorAndFloat("Vector (RGB)"=<1 1 1>, "Float (Alpha)"=$Particles.FadeOffAlpha)
    )) {enabled=false, isSetParameters=true}
   &Material(Base Texture Time:NiagaraFloat=&Multiply_Float(A=$Emitter.Age, B=-1)) {enabled=false}
   &DynamicMaterialParameters(
    Index 0 Param 1=1
    Index 0 Param 2=&Multiply_Float(A=0, B=$Engine.Time)
    Index 0 Param 3=0.5
    Index 0 Param 4=0.5
    Index 1 Param 1=&Multiply_Float(A=-0.25, B=$Emitter.Age)
    Index 1 Param 2=0.25
    Index 1 Param 3=1.2
    Index 1 Param 4=2
    Index 2 Param 1=0.2
    Index 2 Param 2=0.3
    Index 3 Param 1=1
    Index 3 Param 2=1
    Index 3 Param 3=1
    Index 3 Param 4=1
    Module.Param0WriteEnabled=true
    Module.Param0WriteEnabled1=true
    Module.Param0WriteEnabled2=true
    Module.Param0WriteEnabled3=true
    Module.Param1WriteEnabled=true
    Module.Param1WriteEnabled1=true
    Module.Param1WriteEnabled2=true
    Module.Param1WriteEnabled3=true
    Module.Param2WriteEnabled=true
    Module.Param2WriteEnabled1=true
    Module.Param2WriteEnabled2=true
    Module.Param2WriteEnabled3=true
    Module.Param3WriteEnabled=true
    Module.Param3WriteEnabled1=true
    Module.Param3WriteEnabled2=true
    Module.Param3WriteEnabled3=true
    Write Parameter Index 1=true
    Write Parameter Index 2=true
    Write Parameter Index 3=true
   )
   /SetVariables(Particles.RibbonWidth=&FloatFromCurve(FloatCurve=DataInterface() {propertyValues {Curve {keys [<at 0 0>,<at 0.1873 1>,<at 1 2>]}}}, Scale Curve=50)) {isSetParameters=true}
   &Color(Scale Alpha=&FloatFromCurve(
     FloatCurve=DataInterface() {propertyValues {Curve {keys [
         {arriveTangent=2.7939, interpMode=RCIM_Cubic, leaveTangent=2.7939, tangentMode=RCTM_User, time=-0.0046, value=1.0113}
         {arriveTangent=-1.0036, interpMode=RCIM_Cubic, leaveTangent=-4.3444, tangentMode=RCTM_Break, time=0.1387, value=1.232}
         <at 1.0011 0.002>
        ]}}}
     Scale Curve=0.8
    )) {enabled=false}
  ]
  rendererClasses [&NiagaraRibbonRendererProperties]
  renderers [{rendererClass=&NiagaraRibbonRendererProperties}]
 }
 FN_Sprite {
  bEnabled=false
  emitterUpdateScript [
   &EmitterState(Inactive Response=::NewEnumerator1, Life Cycle Mode=::NewEnumerator1, Loop Behavior=::NewEnumerator1, Loop Duration=3, UseLoopDelay=false)
   &SpawnRate(SpawnRate=&Multiply_Float(A=15, B=$User.Sparks), Use Spawn Probability=true)
  ]
  particleSpawnScript [
   &InitializeParticle(
    Lifetime=&UniformRangedFloat(Minimum=0.35, Maximum=1)
    Module.Write Color=true
    Module.Write Lifetime=true
    Module.Write Mass=true
    Module.Write Material Random=true
    Module.Write Position=false
    Module.Write Scale=false
    Module.Write SpriteRotation=false
    Module.Write SpriteSize=true
    Position=&SimulationPosition()
    Sprite Rotation=&UniformRangedFloat(Minimum=0, Maximum=360)
    Sprite Size=&Vector2DFromFloat(Value=&UniformRangedFloat(Minimum=200, Maximum=500))
   )
   &CameraOffset(Camera Offset Amount=&UniformRangedFloat(Minimum=25, Maximum=25))
   &SphereLocation(
    MaskSpawnGroup=false
    Module.Override Randomness=false
    Sphere Origin:Vector3f=$Particles.SampleSkeletalMeshSkeleton.SampledPosition
    Sphere Radius=150
    Surface Only=false
   )
  ]
  particleUpdateScript [
   &ParticleState()
   &SampleSkeletalMeshSkeleton(
    Mesh=DataInterface() {propertyValues {
      DefaultMesh=null
      ExcludeBoneName=root
      OptionalComponentClass=null
      PreviewMesh=&Donut
      SkinningMode=SkinOnTheFly
      SoftSourceActor=null
      SourceMode=Default
      UvSetIndex=0
      WholeMeshLOD=-1
      bExcludeBone=false
      bReadDeformedGeometry=true
      bRequireCurrentFrameData=true
      ComponentTags []
      FilteredBones []
      FilteredSockets [FX_Horn]
      MeshUserParameter {parameter {name=None, typeDefHandle {registeredTypeIndex=1}, varData []}}
      SamplingRegions []
     }}
    Spawn Mode:ENiagaraSkeletalSpawning=::NewEnumerator1
   )
   &SkeletalMeshSkeletonLocation(Inherit Velocity From Mesh=0.15, Module.Inherit Velocity=true, Module.Use Sampled Bone as Ribbon ID=true)
   /SetVariables(Particles.GradientColor:Vector3f=&MakeVectorFromLinearColorRGB(Color=&MakeLinearColorFromVectorAndFloat(
      "Vector (RGB)"=&MakeVectorFromLinearColorRGB(Color=&ColorFromCurve(ColorCurve=DataInterface() {propertyValues {
          AlphaCurve {keys [<at -0.0023 1 RCIM_Cubic>,<at 0.9977 1 RCIM_Cubic>]}
          BlueCurve {keys [<at 0.1366 0 RCIM_Cubic>,<at 0.264 0>,<at 0.4507 0>,<at 0.5951 1>,<at 0.7691 1.5>,<at 0.9071 1.5678 RCIM_Cubic>]}
          GreenCurve {keys [<at 0.1366 0.039 RCIM_Cubic>,<at 0.264 0.8559>,<at 0.4507 1>,<at 0.5951 0.6207>,<at 0.7691 0.4532>,<at 0.9071 0 RCIM_Cubic>]}
          RedCurve {keys [<at 0.1366 1 RCIM_Cubic>,<at 0.264 2>,<at 0.4507 0.3535>,<at 0.5951 0.0629>,<at 0.7691 0>,<at 0.9071 2 RCIM_Cubic>]}
         }}))
      "Float (Alpha)"=1
     ))) {isSetParameters=true, module=SetVariables_#2}
   /SetVariables(Particles.Color=&MakeLinearColorFromVectorAndFloat(
     "Vector (RGB)"=&Lerp_Vector(
      A=&Multiply_VectorByFloat(Vector=$Particles.GradientColor, Float=1)
      B=&Multiply_VectorByFloat(Vector=&UniformRangedVector(Minimum=$Particles.GradientColor, Maximum=$Particles.GradientColor), Float=2)
      Alpha=&FloatFromCurve(FloatCurve=DataInterface() {propertyValues {Curve {keys [
           {interpMode=RCIM_Cubic, tangentMode=RCTM_User, time=-0.0098, value=0.1132}
           {arriveTangent=0.0255, interpMode=RCIM_Cubic, leaveTangent=0.0255, tangentMode=RCTM_User, time=0.2458, value=0.1945}
           {arriveTangent=0.1228, interpMode=RCIM_Cubic, leaveTangent=0.1228, tangentMode=RCTM_User, time=0.5453, value=0.6822}
           {interpMode=RCIM_Cubic, tangentMode=RCTM_User, time=1.0011, value=0.002}
          ]}}})
     )
     "Float (Alpha)"=&FloatFromCurve(FloatCurve=DataInterface() {propertyValues {Curve {keys [
          {interpMode=RCIM_Cubic, tangentMode=RCTM_User, time=-0.0063, value=0.0184}
          {arriveTangent=1.3376, interpMode=RCIM_Cubic, leaveTangent=1.3376, tangentMode=RCTM_User, time=0.096, value=0.9808}
          {arriveTangent=0.3112, interpMode=RCIM_Cubic, leaveTangent=0.3112, tangentMode=RCTM_User, time=0.5407, value=1.3419}
          {arriveTangent=-3.7119, interpMode=RCIM_Cubic, leaveTangent=-3.7119, tangentMode=RCTM_User, time=0.7242, value=0.3749}
          {interpMode=RCIM_Cubic, tangentMode=RCTM_User, time=1.0011, value=0.002}
         ]}}})
    )) {isSetParameters=true}
   &ScaleSpriteSize(Scale Factor=&Vector2DFromFloat(Value=&FloatFromCurve(FloatCurve=DataInterface() {propertyValues {Curve {keys [
          {arriveTangent=2.3096, interpMode=RCIM_Cubic, leaveTangent=2.3096, tangentMode=RCTM_User, time=0, value=0}
          {arriveTangent=-0.0952, interpMode=RCIM_Cubic, leaveTangent=-0.0952, tangentMode=RCTM_User, time=0.431, value=0.3973}
          <at 0.9969 0.0123 RCIM_Cubic>
         ]}}})))
   &SolveForcesAndVelocity(Clamp Velocity=false, Limit Acceleration=false)
   /SetVariables(Particles.Color2=&MakeLinearColorFromVectorAndFloat(
     "Vector (RGB)"=&Lerp_Vector(
      A=&Multiply_VectorByFloat(Vector=$Particles.GradientColor, Float=1)
      B=&Multiply_VectorByFloat(Vector=&UniformRangedVector(Minimum=$Particles.GradientColor, Maximum=$Particles.GradientColor), Float=2)
      Alpha=&FloatFromCurve(FloatCurve=DataInterface() {propertyValues {Curve {keys [
           {interpMode=RCIM_Cubic, tangentMode=RCTM_User, time=-0.0008, value=0.5048}
           {arriveTangent=0.0255, interpMode=RCIM_Cubic, leaveTangent=0.0255, tangentMode=RCTM_User, time=0.0802, value=0.5788}
           {arriveTangent=0.1228, interpMode=RCIM_Cubic, leaveTangent=0.1228, tangentMode=RCTM_User, time=0.5421, value=2.0946}
           {interpMode=RCIM_Cubic, tangentMode=RCTM_User, time=1.0011, value=0.002}
          ]}}})
     )
     "Float (Alpha)"=&FloatFromCurve(
      FloatCurve=DataInterface() {propertyValues {Curve {keys [
          {interpMode=RCIM_Cubic, tangentMode=RCTM_User, time=-0.0063, value=0.0184}
          {arriveTangent=0.3112, interpMode=RCIM_Cubic, leaveTangent=0.3112, tangentMode=RCTM_User, time=0.3734, value=0.0777}
          {arriveTangent=-1.0034, interpMode=RCIM_Cubic, leaveTangent=-1.0034, tangentMode=RCTM_User, time=0.6422, value=0.1616}
          {interpMode=RCIM_Cubic, tangentMode=RCTM_User, time=1.0011, value=0.002}
         ]}}}
      Scale Curve=0.5
     )
    )) {isSetParameters=true}
  ]
  rendererClasses [&NiagaraSpriteRendererProperties]
  renderers [
   {rendererClass=&NiagaraSpriteRendererProperties}
   {rendererClass=&NiagaraSpriteRendererProperties, rendererIndex=1}
  ]
 }
 Light {
  emitterUpdateScript [
   &EmitterState(Inactive Response=::NewEnumerator1, Life Cycle Mode=::NewEnumerator1, Loop Behavior=::NewEnumerator1, Loop Duration=3, UseLoopDelay=false)
   &SpawnBurst_Instantaneous(Spawn Count=1, Use Spawn Probability=false)
  ]
  particleSpawnScript [
   &InitializeParticle(
    Lifetime=&UniformRangedFloat(Minimum=1, Maximum=1)
    Module.Write Color=true
    Module.Write Lifetime=true
    Module.Write Mass=true
    Module.Write Material Random=true
    Module.Write Position=false
    Module.Write Scale=false
    Module.Write SpriteRotation=false
    Module.Write SpriteSize=true
    Position=&SimulationPosition()
    Sprite Rotation=&UniformRangedFloat(Minimum=0, Maximum=360)
    Sprite Size=&Vector2DFromFloat(Value=&UniformRangedFloat(Minimum=1, Maximum=1))
   )
   &SphereLocation(MaskSpawnGroup=false, Module.Override Randomness=false, Offset=<0 -250 0>, Sphere Radius=150, Surface Only=false) {enabled=false}
  ]
  particleUpdateScript [
   &ParticleState()
   &SampleSkeletalMeshSkeleton(
    Mesh=DataInterface() {propertyValues {
      DefaultMesh=null
      ExcludeBoneName=root
      OptionalComponentClass=null
      PreviewMesh=&Donut
      SkinningMode=SkinOnTheFly
      SoftSourceActor=null
      SourceMode=Default
      UvSetIndex=0
      WholeMeshLOD=-1
      bExcludeBone=false
      bReadDeformedGeometry=true
      bRequireCurrentFrameData=true
      ComponentTags []
      FilteredBones []
      FilteredSockets [FX_Horn]
      MeshUserParameter {parameter {name=None, typeDefHandle {registeredTypeIndex=1}, varData []}}
      SamplingRegions []
     }}
    Spawn Mode:ENiagaraSkeletalSpawning=::NewEnumerator1
   ) {module=SampleSkeletalMeshSkeleton001}
   &SkeletalMeshSkeletonLocation(Module.Inherit Velocity=false, Module.Use Sampled Bone as Ribbon ID=true) {module=SkeletalMeshSkeletonLocation001}
   /SetVariables(Particles.GradientColor:Vector3f=&MakeVectorFromLinearColorRGB(Color=&MakeLinearColorFromVectorAndFloat(
      "Vector (RGB)"=&MakeVectorFromLinearColorRGB(Color=&ColorFromCurve(ColorCurve=DataInterface() {propertyValues {
          AlphaCurve {keys [<at -0.0023 1 RCIM_Cubic>,<at 0.9977 1 RCIM_Cubic>]}
          BlueCurve {keys [<at -0.0023 0 RCIM_Cubic>,<at 0.1706 0>,<at 0.4507 0>,<at 0.5951 1>,<at 0.7691 1.5>,<at 0.9071 1.5678 RCIM_Cubic>]}
          GreenCurve {keys [<at -0.0023 0.039 RCIM_Cubic>,<at 0.1706 0.8559>,<at 0.4507 1>,<at 0.5951 0.6207>,<at 0.7691 0.4532>,<at 0.9071 0 RCIM_Cubic>]}
          RedCurve {keys [<at -0.0023 1 RCIM_Cubic>,<at 0.1706 2>,<at 0.4507 0.3535>,<at 0.5951 0.0629>,<at 0.7691 0>,<at 0.9071 2 RCIM_Cubic>]}
         }}))
      "Float (Alpha)"=1
     ))) {isSetParameters=true, module=SetVariables_#2}
   /SetVariables(Particles.Color=&MakeLinearColorFromVectorAndFloat(
     "Vector (RGB)"=&Lerp_Vector(A=$Particles.GradientColor, B=&UniformRangedVector(Minimum=<15 15 15>, Maximum=<250 250 250>), Alpha=0)
     "Float (Alpha)"=&FloatFromCurve(
      FloatCurve=DataInterface() {propertyValues {Curve {keys [
          {interpMode=RCIM_Cubic, tangentMode=RCTM_User, time=-0.002, value=0.0241}
          {arriveTangent=-0.0221, interpMode=RCIM_Cubic, leaveTangent=-0.0221, tangentMode=RCTM_User, time=0.0954, value=0.9231}
          <at 1.0011 0.002>
         ]}}}
      Scale Curve=0.8
     )
    )) {enabled=false, isSetParameters=true}
   /SetVariables(Particles.Color=&MakeLinearColorFromVectorAndFloat(
     "Vector (RGB)"=&Lerp_Vector(
      A=&Multiply_VectorByFloat(Vector=$Particles.GradientColor, Float=25)
      B=&Multiply_VectorByFloat(Vector=$Particles.GradientColor, Float=250)
      Alpha=0.5
     )
     "Float (Alpha)"=&FloatFromCurve(FloatCurve=DataInterface() {propertyValues {Curve {keys [
          {interpMode=RCIM_Cubic, tangentMode=RCTM_User, time=0, value=0}
          {arriveTangent=-0.8483, interpMode=RCIM_Cubic, leaveTangent=-0.8483, tangentMode=RCTM_User, time=0.5, value=1}
          {interpMode=RCIM_Cubic, tangentMode=RCTM_User, time=1, value=0}
         ]}}})
    )) {isSetParameters=true}
   &SolveForcesAndVelocity(Clamp Velocity=false, Limit Acceleration=false) {enabled=false}
   /Game/Effects/Niagara/Modules/Light/FN_Light_v5/FN_Light(Light Exponent=2, Light Radius=100, LightEnabled=true, Scale Color=200) {enabled=false}
   /SetVariables(Particles.LightPosition:NiagaraPosition=$Particles.Position) {enabled=false, isSetParameters=true}
   /SetVariables(Particles.LightExponent=1) {isSetParameters=true}
   /SetVariables() {isSetParameters=true}
   &SystemLocation(MaskSpawnGroup=false, Offset=<0 -50 150>, Offset Coordinate Space=::Simulation) {enabled=false}
  ]
  rendererClasses [&NiagaraLightRendererProperties]
  renderers [{rendererClass=&NiagaraLightRendererProperties, rendererIndex=0}]
 }
 Pop_Sparks {
  bEnabled=false
  emitterUpdateScript [
   &EmitterState(Inactive Response=::NewEnumerator1, Life Cycle Mode=::NewEnumerator1, Loop Behavior=::NewEnumerator1, Loop Duration=3, UseLoopDelay=false)
   &SpawnRate(SpawnRate=100, Use Spawn Probability=true)
  ]
  particleSpawnScript [
   &InitializeParticle(
    Color=&LerpLinearColors(StartColor=<rgba 1 1 1 1>, EndColor=<rgba 1 1 1 1>, LerpFactor=1)
    Lifetime=&Multiply_Float(A=&UniformRangedFloat(Minimum=0.5, Maximum=0.7), B=1)
    Module.Write Color=true
    Module.Write Lifetime=true
    Module.Write Mass=true
    Module.Write Material Random=true
    Module.Write Position=false
    Module.Write Scale=false
    Module.Write SpriteRotation=false
    Module.Write SpriteSize=true
    Position=&SimulationPosition()
    Sprite Size=&UniformRangedVector2D(
     Minimum=&Vector2DFromFloat(Value=20)
     Maximum=&Vector2DFromFloat(Value=35)
    )
   )
   &SampleSkeletalMeshSkeleton(Mesh:NiagaraDataInterfaceSkeletalMesh=$User.SkeletalMesh)
   &SkeletalMeshSkeletonLocation(
    Module.Inherit Velocity=false
    Module.Use Sampled Bone as Ribbon ID=true
    Sampled Position Offset:Vector3f=$Output.SphereLocation.SphereNormal
   )
   &SphereLocation(
    MaskSpawnGroup=false
    Module.Override Randomness=false
    Sphere Radius=&UniformRangedFloat(Minimum=15, Maximum=15)
    Surface Only=false
   )
   &CurlNoiseForce(
    Noise Frequency=15
    Noise Strength=&Multiply_Float(
     A=&FloatFromCurve(
      FloatCurve=DataInterface() {propertyValues {Curve {keys [<at -0.0046 1.0113>,<at 1.0011 0.002>]}}}
      CurveIndex=$Emitter.NormalizedLoopAge
      Scale Curve=0.25
     )
     B=&MakeFloatFromVector(Vector=$Engine.Owner.Velocity, Channel=::NewEnumerator0)
    )
    Pan Noise=true
    Pan Noise Field=<0 10 0>
   ) {module=CurlNoiseForce001}
   &ApplyInitialForces()
  ]
  particleUpdateScript [
   &ParticleState()
   /SetVariables(Particles.GradientColor:Vector3f=&MakeVectorFromLinearColorRGB(Color=&MakeLinearColorFromVectorAndFloat(
      "Vector (RGB)"=&MakeVectorFromLinearColorRGB(Color=&ColorFromCurve(ColorCurve=DataInterface() {propertyValues {
          AlphaCurve {keys [<at -0.0023 1 RCIM_Cubic>,<at 0.9977 1 RCIM_Cubic>]}
          BlueCurve {keys [<at 0.1175 0 RCIM_Cubic>,<at 0.2682 0>,<at 0.4507 0>,<at 0.5951 1>,<at 0.7691 1.5>,<at 0.9071 1.5678 RCIM_Cubic>]}
          GreenCurve {keys [<at 0.1175 0.039 RCIM_Cubic>,<at 0.2682 0.8559>,<at 0.4507 1>,<at 0.5951 0.6207>,<at 0.7691 0.4532>,<at 0.9071 0 RCIM_Cubic>]}
          RedCurve {keys [<at 0.1175 1 RCIM_Cubic>,<at 0.2682 2>,<at 0.4507 0.3535>,<at 0.5951 0.0629>,<at 0.7691 0>,<at 0.9071 2 RCIM_Cubic>]}
         }}))
      "Float (Alpha)"=1
     ))) {isSetParameters=true}
   /SetVariables(Particles.Color=&MakeLinearColorFromVectorAndFloat(
     "Vector (RGB)"=&Lerp_Vector(
      A=&Multiply_VectorByFloat(Vector=$Particles.GradientColor, Float=5)
      B=&Multiply_VectorByFloat(Vector=$Particles.GradientColor, Float=25)
      Alpha=0.5
     )
     "Float (Alpha)"=&FloatFromCurve(
      FloatCurve=DataInterface() {propertyValues {Curve {keys [
          {interpMode=RCIM_Cubic, tangentMode=RCTM_User, time=-0.0063, value=0.0184}
          {arriveTangent=-0.8483, interpMode=RCIM_Cubic, leaveTangent=-0.8483, tangentMode=RCTM_User, time=0.1718, value=1.7147}
          {arriveTangent=-5.0978, interpMode=RCIM_Cubic, leaveTangent=-5.0978, tangentMode=RCTM_User, time=0.2818, value=1.2143}
          {arriveTangent=-0.2783, interpMode=RCIM_Cubic, leaveTangent=-0.2783, tangentMode=RCTM_User, time=0.4294, value=1.1587}
          {arriveTangent=0.3112, interpMode=RCIM_Cubic, leaveTangent=0.3112, tangentMode=RCTM_User, time=0.5407, value=1.3419}
          {arriveTangent=-3.7119, interpMode=RCIM_Cubic, leaveTangent=-3.7119, tangentMode=RCTM_User, time=0.7242, value=0.3749}
          {interpMode=RCIM_Cubic, tangentMode=RCTM_User, time=1.0011, value=0.002}
         ]}}}
      Scale Curve=0.8
     )
    )) {isSetParameters=true, module=SetVariables_#1}
   &AddVelocity(Velocity=&RotateVectorByQuaternion(
     Vector To Rotate=&UniformRangedVector(Minimum=<-15 5 -15>, Maximum=<15 25 15>)
     Quaternion=$Particles.SampleSkeletalMeshSkeleton.SampledRotation
    )) {enabled=false}
   &AccelerationForce(Acceleration=<0 0 -100>) {enabled=false}
   &CurlNoiseForce(
    Noise Frequency=1
    Noise Strength=&Multiply_Float(A=0.25, B=&MakeFloatFromVector(Vector=$Engine.Owner.Velocity, Channel=::NewEnumerator0))
    Pan Noise=true
    Pan Noise Field=<0 10 0>
   ) {enabled=false}
   &Drag(Drag=&UniformRangedFloat(Minimum=1, Maximum=5), Use Linear Drag=true, Use Rotational Drag=false) {enabled=false}
   &InheritVelocity(
    Module.Limit Speed=true
    Velocity Scale=<0 -0.025 0>
    Velocity Source=&TransformVector(Vector=$Engine.Owner.Velocity, Source Space=::World, Destination Space=::Local)
   )
   &SolveForcesAndVelocity(Clamp Velocity=false, Limit Acceleration=false)
   &ScaleSpriteSize(Scale Factor=&Vector2DFromFloat(Value=&FloatFromCurve(FloatCurve=DataInterface() {propertyValues {Curve {keys [
          {arriveTangent=7.1136, interpMode=RCIM_Cubic, leaveTangent=7.1136, tangentMode=RCTM_User, time=0, value=0}
          {arriveTangent=1.4654, interpMode=RCIM_Cubic, leaveTangent=1.4654, time=0.18, value=1}
          <at 0.3814 0.5589>
          {arriveTangent=-0.0952, interpMode=RCIM_Cubic, leaveTangent=-0.0952, tangentMode=RCTM_User, time=0.7849, value=0.7663}
          <at 1 0.5 RCIM_Cubic>
         ]}}})))
   &ScaleSpriteSizeBySpeed(Max Scale Factor=<0.5 0.5>)
   &CameraOffset(Camera Offset Amount=50)
  ]
  rendererClasses [&NiagaraSpriteRendererProperties]
  renderers [{rendererClass=&NiagaraSpriteRendererProperties}]
 }
}
summary {
 systemName=NS_Glider_Donut_Deploy_FrontEnd
 emitters [
  {bEnabled=false, emitterName=Pop_Sparks, rendererClasses [&NiagaraSpriteRendererProperties]}
  {bEnabled=false, emitterName=FN_Sprite, rendererClasses [&NiagaraSpriteRendererProperties]}
  {bEnabled=false, emitterName=FN_Contrails_Ribbon_v4_A001, rendererClasses [&NiagaraRibbonRendererProperties]}
  {emitterName=Light, rendererClasses [&NiagaraLightRendererProperties]}
  {bEnabled=false, emitterName=Confetti, rendererClasses [&NiagaraSpriteRendererProperties]}
 ]
 userVariables [
  User.RibbonIDOffset {type=NiagaraInt32}
  User.SkeletalMesh {
   type=NiagaraDataInterfaceSkeletalMesh
   defaultValue {
    struct=NiagaraExt_VariableValue_DataInterface
    value {dataInterface=&self:NiagaraDataInterfaceSkeletalMesh_0}
   }
  }
  User.Sparks {type=NiagaraFloat, defaultValue=1, description=""}
 ]
}
