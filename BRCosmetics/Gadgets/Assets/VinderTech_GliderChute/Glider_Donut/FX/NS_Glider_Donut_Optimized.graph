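# Niagara system for the Donut glider cosmetic, trailing rainbow-gradient confetti, sparks, glow sprites, cloudy contrails and a colored light from the glider's skeletal mesh.
# Spawn rates ramp with System.SpawnRamp, User.Sparks scales the glow sprites, and emitters place particles by sampling User.SkeletalMesh.
# Owner velocity drives inherited motion and curl noise on confetti and sparks.
# FrontEnd modifiers adjust sizes, alpha and light brightness between gameplay and front end.
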NiagaraSystem /BRCosmetics/Gadgets/Assets/VinderTech_GliderChute/Glider_Donut/FX/NS_Glider_Donut_Optimized
dependencies {
 usedDataInterfaces [&NiagaraDataInterfaceColorCurve, &NiagaraDataInterfaceCurve]
 usedDynamicInputs []
 usedModules [
  &self:SystemSpawnScript.SystemScriptSource.SystemScriptGraph.NiagaraNodeAssignment_0.SetVariables_#1
  &SystemState
  &self:SystemSpawnScript.SystemScriptSource.SystemScriptGraph.NiagaraNodeAssignment_1.SetVariables_#2
 ]
}
emitters {
 Confetti {
  emitterUpdateScript [
   &EmitterState(Life Cycle Mode=::NewEnumerator1, Loop Duration=1, UseLoopDelay=false)
   &SpawnRate(SpawnRate=&Lerp_Float(A=0, B=100, Alpha=$System.SpawnRamp), Use Spawn Probability=true)
  ]
  particleSpawnScript [
   /Niagara/Modules/Spawn/Initialization/InitializeParticle(
    Lifetime=&Multiply_Float(A=&UniformRangedFloat(Minimum=0.5, Maximum=1), B=1)
    Module.Write Color=true
    Module.Write Lifetime=true
    Module.Write Mass=true
    Module.Write Material Random=true
    Module.Write Position=false
    Module.Write Scale=false
    Module.Write SpriteRotation=true
    Module.Write SpriteSize=true
    Position=&SimulationPosition()
    Sprite Rotation=&UniformRangedFloat(Minimum=0, Maximum=360)
    Sprite Size=&UniformRangedVector2D(
     Minimum=&Vector2DFromFloat(Value=&FN_FrontEndFloatModifier(Gameplay=45, Front End=25))
     Maximum=&Vector2DFromFloat(Value=&FN_FrontEndFloatModifier(Gameplay=65, Front End=35))
    )
   )
   &SampleSkeletalMeshSkeleton(Mesh:NiagaraDataInterfaceSkeletalMesh=$User.SkeletalMesh)
   /Niagara/Modules/Spawn/Location/SkeletalMeshSkeletonLocation(Inherit Velocity From Mesh=150, Module.Inherit Velocity=false, Module.Use Sampled Bone as Ribbon ID=true) {enabled=false}
   /Niagara/Modules/Spawn/Location/SphereLocation(
    MaskSpawnGroup=false
    Module.Override Randomness=false
    Offset=&FN_FrontEndVector3Modifier(Gameplay=<0 0 0>, Front End=<0 0 0>)
    Sphere Origin:Vector3f=$Particles.SampleSkeletalMeshSkeleton.SampledPosition
    Sphere Radius=&UniformRangedFloat(Minimum=15, Maximum=30)
    Surface Only=false
   )
   /Niagara/Modules/Update/Forces/CurlNoiseForce(
    Noise Frequency=25
    Noise Strength=&Multiply_Float(A=0.2, B=&MakeFloatFromVector(Vector=$Engine.Owner.Velocity, Channel=::NewEnumerator0))
    Pan Noise=true
    Pan Noise Field=<0 1 0>
   ) {enabled=false, module=CurlNoiseForce001}
   &AddVelocity(Velocity=&RotateVectorByQuaternion(
     Vector To Rotate=&UniformRangedVector(Minimum=<-1 -1 -1>, Maximum=<1 1 1>)
     Quaternion=$Particles.SampleSkeletalMeshSkeleton.SampledRotation
    ))
   &ApplyInitialForces()
   /SetVariables() {isSetParameters=true}
  ]
  particleUpdateScript [
   &ParticleState()
   /SetVariables(Particles.GradientColor:Vector3f=&MakeVectorFromLinearColorRGB(Color=&ColorFromCurve(ColorCurve=DataInterface() {propertyValues {
        AlphaCurve {keys [<at 0.2533 1 RCIM_Cubic>,<at 0.9977 1 RCIM_Cubic>]}
        BlueCurve {keys [<at 0.4507 0 RCIM_Cubic>,<at 0.5102 0>,<at 0.6354 0>,<at 0.7224 1>,<at 0.7946 1.5>,<at 0.9219 1.5678 RCIM_Cubic>]}
        GreenCurve {keys [<at 0.4507 0.039 RCIM_Cubic>,<at 0.5102 0.8559>,<at 0.6354 1>,<at 0.7224 0.6207>,<at 0.7946 0.4532>,<at 0.9219 0 RCIM_Cubic>]}
        RedCurve {keys [<at 0.4507 1 RCIM_Cubic>,<at 0.5102 2>,<at 0.6354 0.3535>,<at 0.7224 0.0629>,<at 0.7946 0>,<at 0.9219 2 RCIM_Cubic>]}
       }}))) {isSetParameters=true}
   /SetVariables(Particles.Color=&MakeLinearColorFromVectorAndFloat(
     "Vector (RGB)"=&Lerp_Vector(
      A=<1 1 1>
      B=$Particles.GradientColor
      Alpha=&FloatFromCurve(FloatCurve=DataInterface() {propertyValues {Curve {keys [
           {interpMode=RCIM_Cubic, tangentMode=RCTM_User, time=-0.0098, value=0.1132}
           {arriveTangent=0.0255, interpMode=RCIM_Cubic, leaveTangent=0.0255, tangentMode=RCTM_User, time=0.2561, value=0.1643}
           {arriveTangent=0.1228, interpMode=RCIM_Cubic, leaveTangent=0.1228, tangentMode=RCTM_User, time=0.5453, value=0.6822}
           {interpMode=RCIM_Cubic, tangentMode=RCTM_User, time=1.0011, value=0.002}
          ]}}})
     )
     "Float (Alpha)"=&FloatFromCurve(FloatCurve=DataInterface() {propertyValues {Curve {keys [
          {interpMode=RCIM_Cubic, tangentMode=RCTM_User, time=-0.0063, value=0.0184}
          {arriveTangent=1.3376, interpMode=RCIM_Cubic, leaveTangent=1.3376, tangentMode=RCTM_User, time=0.096, value=0.9808}
          {arriveTangent=0.3112, interpMode=RCIM_Cubic, leaveTangent=0.3112, tangentMode=RCTM_User, time=0.5407, value=1.3419}
          {arriveTangent=-3.7119, interpMode=RCIM_Cubic, leaveTangent=-3.7119, tangentMode=RCTM_User, time=0.7242, value=0.3749}
          {interpMode=RCIM_Cubic, tangentMode=RCTM_User, time=1.0011, value=0.002}
         ]}}})
    )) {isSetParameters=true, module=SetVariables_#3}
   &AccelerationForce(Acceleration=<0 0 -800>)
   &CurlNoiseForce(
    Noise Frequency=1
    Noise Strength=&Multiply_Float(A=1, B=&MakeFloatFromVector(Vector=$Engine.Owner.Velocity, Channel=::NewEnumerator0))
    Pan Noise=true
    Pan Noise Field=<0 10 0>
   )
   &Drag(Drag=$Particles.RandomDrag, Use Linear Drag=true, Use Rotational Drag=false)
   &InheritVelocity(
    Module.Limit Speed=true
    Velocity Scale=<-0.0005 -0.025 -0.02>
    Velocity Source=&TransformVector(Vector=$Engine.Owner.Velocity, Source Space=::World, Destination Space=::Local)
   )
   &SolveForcesAndVelocity(Clamp Velocity=false, Limit Acceleration=false)
   &ScaleSpriteSize(Scale Factor=&Vector2DFromFloat(Value=&FloatFromCurve(FloatCurve=DataInterface() {propertyValues {Curve {keys [
          {arriveTangent=2.3096, interpMode=RCIM_Cubic, leaveTangent=2.3096, tangentMode=RCTM_User, time=0, value=0}
          {arriveTangent=-0.0952, interpMode=RCIM_Cubic, leaveTangent=-0.0952, tangentMode=RCTM_User, time=0.431, value=0.3973}
          <at 0.9969 0.0123 RCIM_Cubic>
         ]}}})))
   &ScaleSpriteSizeBySpeed(Max Scale Factor=<0.5 0.5>)
   &CameraOffset(Camera Offset Amount=25)
   /Niagara/Modules/Update/SubUV/SubUVAnimation(Number Of Frames=36)
  ]
  rendererClasses [&NiagaraSpriteRendererProperties]
  renderers [{rendererClass=&NiagaraSpriteRendererProperties}]
 }
 FN_Contrails_SpriteCloudy {
  emitterSpawnScript [&FN_PreserveSkydiveUserVariables() {module=PreserveUserVariables}]
  emitterUpdateScript [
   &EmitterState(
    Enable Distance Culling=true
    Enable Visibility Culling=true
    MaxDistance=1500
    Module.bUseMaxDistance=true
    Module.bUseMinDistance=false
    Scalability Mode=::NewEnumerator1
   )
   &SpawnRate(SpawnRate=&Lerp_Float(A=0, B=25, Alpha=$System.SpawnRamp), Use Spawn Probability=true)
  ]
  particleSpawnScript [
   &InitializeParticle(
    Color=&UniformRangedLinearColor(Minimum=<rgba 1 1 1 0.25>, Maximum=<rgba 1 1 1 0.25>)
    Lifetime=&UniformRangedFloat(Minimum=0.5, Maximum=0.6)
    Module.Write Color=true
    Module.Write Lifetime=true
    Module.Write Mass=true
    Module.Write Material Random=true
    Module.Write Position=false
    Module.Write Scale=false
    Module.Write SpriteRotation=true
    Module.Write SpriteSize=true
    Position=&SimulationPosition()
    Sprite Rotation=&UniformRangedFloat(Minimum=-90, Maximum=90)
    Sprite Size=&UniformRangedVector2D(Minimum=<10 20>, Maximum=<10 50>)
   )
   &SampleSkeletalMeshSkeleton(Mesh:NiagaraDataInterfaceSkeletalMesh=$User.SkeletalMesh)
   &SkeletalMeshSkeletonLocation(Inherit Velocity From Mesh=150, Module.Inherit Velocity=false, Module.Use Sampled Bone as Ribbon ID=true) {enabled=false}
   &SphereLocation(
    MaskSpawnGroup=false
    Module.Override Randomness=false
    Sphere Origin:Vector3f=$Particles.SampleSkeletalMeshSkeleton.SampledPosition
    Sphere Radius=&UniformRangedFloat(Minimum=10, Maximum=25)
    Surface Only=false
   )
   &DynamicMaterialParameters(
    Index 0 Param 1=&UniformRangedFloat(Minimum=0, Maximum=3)
    Module.Param0WriteEnabled=true
    Module.Param1WriteEnabled=false
    Module.Param2WriteEnabled=false
    Module.Param3WriteEnabled=false
   )
   &AddVelocity(Velocity=&UniformRangedVector(Minimum=<0 50 0>, Maximum=<0 100 0>))
  ]
  particleUpdateScript [
   &ParticleState()
   &InheritVelocity(
    Module.Limit Speed=true
    Velocity Scale=<-0.02 -0.025 -0.005>
    Velocity Source=&TransformVector(Vector=$Engine.Owner.Velocity, Source Space=::World, Destination Space=::Local)
   )
   &SolveForcesAndVelocity(Clamp Velocity=false, Limit Acceleration=false)
   &ScaleSpriteSize(Scale Factor=&MakeVector2D(
     X=&Lerp_Float(A=0, B=10, Alpha=$Particles.NormalizedAge)
     Y=&Lerp_Float(A=0, B=12, Alpha=$Particles.NormalizedAge)
    ))
   /SetVariables(Particles.GradientColor:Vector3f=&MakeVectorFromLinearColorRGB(Color=&MakeLinearColorFromVectorAndFloat(
      "Vector (RGB)"=&MakeVectorFromLinearColorRGB(Color=&ColorFromCurve(ColorCurve=DataInterface() {propertyValues {
          AlphaCurve {keys [<at -0.0023 1 RCIM_Cubic>,<at 0.9977 1 RCIM_Cubic>]}
          BlueCurve {keys [<at 0.0942 0 RCIM_Cubic>,<at 0.2809 0>,<at 0.4507 0>,<at 0.5951 1>,<at 0.7691 1.5>,<at 0.9071 1.5678 RCIM_Cubic>]}
          GreenCurve {keys [<at 0.0942 0.039 RCIM_Cubic>,<at 0.2809 0.8559>,<at 0.4507 1>,<at 0.5951 0.6207>,<at 0.7691 0.4532>,<at 0.9071 0 RCIM_Cubic>]}
          RedCurve {keys [<at 0.0942 1 RCIM_Cubic>,<at 0.2809 2>,<at 0.4507 0.3535>,<at 0.5951 0.0629>,<at 0.7691 0>,<at 0.9071 2 RCIM_Cubic>]}
         }}))
      "Float (Alpha)"=1
     ))) {enabled=false, isSetParameters=true}
   /SetVariables(Particles.Color=&MakeLinearColorFromVectorAndFloat(
     "Vector (RGB)"=&Multiply_Vector(
      A=&MakeVectorFromLinearColorRGB(Color=&ColorFromCurve(ColorCurve=DataInterface() {propertyValues {
          AlphaCurve {keys [<at -0.0023 1 RCIM_Cubic>,<at 0.9977 1 RCIM_Cubic>]}
          BlueCurve {keys [<at 0.0942 0 RCIM_Cubic>,<at 0.2809 0>,<at 0.4507 0>,<at 0.5951 1>,<at 0.7691 1.5>,<at 0.9071 1.5678 RCIM_Cubic>]}
          GreenCurve {keys [<at 0.0942 0.039 RCIM_Cubic>,<at 0.2809 0.8559>,<at 0.4507 1>,<at 0.5951 0.6207>,<at 0.7691 0.4532>,<at 0.9071 0 RCIM_Cubic>]}
          RedCurve {keys [<at 0.0942 1 RCIM_Cubic>,<at 0.2809 2>,<at 0.4507 0.3535>,<at 0.5951 0.0629>,<at 0.7691 0>,<at 0.9071 2 RCIM_Cubic>]}
         }}))
      B=&VectorFromFloat(Value=&FloatFromCurve(FloatCurve=DataInterface() {propertyValues {Curve {keys [
            {interpMode=RCIM_Cubic, tangentMode=RCTM_User, time=-0.0057, value=0.477}
            {arriveTangent=-0.1817, interpMode=RCIM_Cubic, leaveTangent=-0.1817, tangentMode=RCTM_User, time=0.068, value=0.7268}
            {interpMode=RCIM_Cubic, tangentMode=RCTM_User, time=0.9643, value=0.5148}
           ]}}}))
     )
     "Float (Alpha)"=&FloatFromCurve(FloatCurve=DataInterface() {propertyValues {Curve {keys [
          {interpMode=RCIM_Cubic, tangentMode=RCTM_User, time=-0.0088, value=0.0144}
          {arriveTangent=-0.1817, interpMode=RCIM_Cubic, leaveTangent=-0.1817, tangentMode=RCTM_User, time=0.068, value=0.7268}
          {interpMode=RCIM_Cubic, tangentMode=RCTM_User, time=0.9384, value=0.0237}
         ]}}})
    )) {enabled=false, isSetParameters=true}
   &Color(
    Color=&ColorFromCurve(ColorCurve=DataInterface() {propertyValues {
       AlphaCurve {keys [<at -0.0023 1 RCIM_Cubic>,<at 0.9977 1 RCIM_Cubic>]}
       BlueCurve {keys [<at 0.0942 0 RCIM_Cubic>,<at 0.2809 0>,<at 0.4507 0>,<at 0.5951 1>,<at 0.7691 1.5>,<at 0.9071 1.5678 RCIM_Cubic>]}
       GreenCurve {keys [<at 0.0942 0.039 RCIM_Cubic>,<at 0.2809 0.8559>,<at 0.4507 1>,<at 0.5951 0.6207>,<at 0.7691 0.4532>,<at 0.9071 0 RCIM_Cubic>]}
       RedCurve {keys [<at 0.0942 1 RCIM_Cubic>,<at 0.2809 2>,<at 0.4507 0.3535>,<at 0.5951 0.0629>,<at 0.7691 0>,<at 0.9071 2 RCIM_Cubic>]}
      }})
    Scale Alpha=&FloatFromCurve(FloatCurve=DataInterface() {propertyValues {Curve {keys [
         {interpMode=RCIM_Cubic, tangentMode=RCTM_User, time=-0.0088, value=0.0144}
         {arriveTangent=-0.1817, interpMode=RCIM_Cubic, leaveTangent=-0.1817, tangentMode=RCTM_User, time=0.068, value=0.7268}
         {interpMode=RCIM_Cubic, tangentMode=RCTM_User, time=0.9384, value=0.0237}
        ]}}})
    Scale Color=&VectorFromFloat(Value=&FloatFromCurve(FloatCurve=DataInterface() {propertyValues {Curve {keys [
          {interpMode=RCIM_Cubic, tangentMode=RCTM_User, time=-0.0057, value=0.477}
          {arriveTangent=-0.1817, interpMode=RCIM_Cubic, leaveTangent=-0.1817, tangentMode=RCTM_User, time=0.068, value=0.7268}
          {interpMode=RCIM_Cubic, tangentMode=RCTM_User, time=0.9643, value=0.5148}
         ]}}}))
   )
   &DynamicMaterialParameters(
    Index 0 Param 2=$Particles.NormalizedAge
    Index 0 Param 3=$Particles.MaterialRandom
    Module.Param0WriteEnabled=false
    Module.Param1WriteEnabled=true
    Module.Param2WriteEnabled=true
    Module.Param3WriteEnabled=true
   ) {module=DynamicMaterialParameters001}
   &CameraOffset(Camera Offset Amount=100)
  ]
  rendererClasses [&NiagaraSpriteRendererProperties]
  renderers [{rendererClass=&NiagaraSpriteRendererProperties}]
 }
 FN_Sprite {
  emitterUpdateScript [
   &EmitterState(Life Cycle Mode=::NewEnumerator1, Loop Duration=1, UseLoopDelay=false)
   &SpawnRate(
    SpawnRate=&Multiply_Float(A=&Lerp_Float(A=0, B=15, Alpha=$System.SpawnRamp), B=$User.Sparks)
    Use Spawn Probability=true
   )
  ]
  particleSpawnScript [&InitializeParticle(
    Lifetime=&UniformRangedFloat(Minimum=0.35, Maximum=1)
    Module.Write Color=true
    Module.Write Lifetime=true
    Module.Write Mass=true
    Module.Write Material Random=true
    Module.Write Position=false
    Module.Write Scale=false
    Module.Write SpriteRotation=false
    Module.Write SpriteSize=true
    Position=&SimulationPosition()
    Sprite Rotation=&UniformRangedFloat(Minimum=0, Maximum=360)
    Sprite Size=&Vector2DFromFloat(Value=&Multiply_Float(
      A=&UniformRangedFloat(Minimum=100, Maximum=250)
      B=&FN_FrontEndFloatModifier(Gameplay=1, Front End=0.4)
     ))
   )]
  particleUpdateScript [
   &ParticleState()
   &SkeletalMeshLocation(
    Inherit Velocity=false
    MaskSpawnGroup=false
    Mesh Sampling Type=::NewEnumerator5
    Skeletal Mesh=$User.SkeletalMesh
    Socket Sampling Mode=::NewEnumerator5
   )
   &CameraOffset(Camera Offset Amount=25)
   /SetVariables(Particles.GradientColor:Vector3f=&MakeVectorFromLinearColorRGB(Color=&ColorFromCurve(ColorCurve=DataInterface() {propertyValues {
        AlphaCurve {keys [<at -0.0023 1 RCIM_Cubic>,<at 0.9977 1 RCIM_Cubic>]}
        BlueCurve {keys [<at 0.1366 0 RCIM_Cubic>,<at 0.264 0>,<at 0.4507 0>,<at 0.5951 1>,<at 0.7691 1.5>,<at 0.9071 1.5678 RCIM_Cubic>]}
        GreenCurve {keys [<at 0.1366 0.039 RCIM_Cubic>,<at 0.264 0.8559>,<at 0.4507 1>,<at 0.5951 0.6207>,<at 0.7691 0.4532>,<at 0.9071 0 RCIM_Cubic>]}
        RedCurve {keys [<at 0.1366 1 RCIM_Cubic>,<at 0.264 2>,<at 0.4507 0.3535>,<at 0.5951 0.0629>,<at 0.7691 0>,<at 0.9071 2 RCIM_Cubic>]}
       }}))) {isSetParameters=true, module=SetVariables_#4}
   /SetVariables(Particles.Color=&MakeLinearColorFromVectorAndFloat(
     "Vector (RGB)"=&Lerp_Vector(
      A=$Particles.GradientColor
      B=&Multiply_VectorByFloat(Vector=$Particles.GradientColor, Float=2)
      Alpha=&FloatFromCurve(FloatCurve=DataInterface() {propertyValues {Curve {keys [
           {interpMode=RCIM_Cubic, tangentMode=RCTM_User, time=-0.0098, value=0.1132}
           {arriveTangent=0.0255, interpMode=RCIM_Cubic, leaveTangent=0.0255, tangentMode=RCTM_User, time=0.2458, value=0.1945}
           {arriveTangent=0.1228, interpMode=RCIM_Cubic, leaveTangent=0.1228, tangentMode=RCTM_User, time=0.5453, value=0.6822}
           {interpMode=RCIM_Cubic, tangentMode=RCTM_User, time=1.0011, value=0.002}
          ]}}})
     )
     "Float (Alpha)"=&FloatFromCurve(
      FloatCurve=DataInterface() {propertyValues {Curve {keys [
          {interpMode=RCIM_Cubic, tangentMode=RCTM_User, time=-0.0063, value=0.0184}
          {arriveTangent=1.3376, interpMode=RCIM_Cubic, leaveTangent=1.3376, tangentMode=RCTM_User, time=0.096, value=0.9808}
          {arriveTangent=0.3112, interpMode=RCIM_Cubic, leaveTangent=0.3112, tangentMode=RCTM_User, time=0.5407, value=1.3419}
          {arriveTangent=-3.7119, interpMode=RCIM_Cubic, leaveTangent=-3.7119, tangentMode=RCTM_User, time=0.7242, value=0.3749}
          {interpMode=RCIM_Cubic, tangentMode=RCTM_User, time=1.0011, value=0.002}
         ]}}}
      Scale Curve=&FN_FrontEndFloatModifier(Gameplay=1, Front End=0.5)
     )
    )) {isSetParameters=true}
   &Color() {enabled=false}
   &ScaleSpriteSize(Scale Factor=&Vector2DFromFloat(Value=&FloatFromCurve(FloatCurve=DataInterface() {propertyValues {Curve {keys [
          {arriveTangent=2.3096, interpMode=RCIM_Cubic, leaveTangent=2.3096, tangentMode=RCTM_User, time=0, value=0}
          {arriveTangent=-0.0952, interpMode=RCIM_Cubic, leaveTangent=-0.0952, tangentMode=RCTM_User, time=0.431, value=0.3973}
          <at 0.9969 0.0123 RCIM_Cubic>
         ]}}}))) {enabled=false}
   &SolveForcesAndVelocity(Clamp Velocity=false, Limit Acceleration=false)
   /SetVariables(Particles.Color2=&MakeLinearColorFromVectorAndFloat(
     "Vector (RGB)"=&Lerp_Vector(
      A=$Particles.GradientColor
      B=&Multiply_VectorByFloat(Vector=$Particles.GradientColor, Float=2)
      Alpha=&FloatFromCurve(FloatCurve=DataInterface() {propertyValues {Curve {keys [
           {interpMode=RCIM_Cubic, tangentMode=RCTM_User, time=-0.0008, value=0.5048}
           {arriveTangent=0.0255, interpMode=RCIM_Cubic, leaveTangent=0.0255, tangentMode=RCTM_User, time=0.0802, value=0.5788}
           {arriveTangent=0.1228, interpMode=RCIM_Cubic, leaveTangent=0.1228, tangentMode=RCTM_User, time=0.5421, value=2.0946}
           {interpMode=RCIM_Cubic, tangentMode=RCTM_User, time=1.0011, value=0.002}
          ]}}})
     )
     "Float (Alpha)"=&FloatFromCurve(
      FloatCurve=DataInterface() {propertyValues {Curve {keys [
          {interpMode=RCIM_Cubic, tangentMode=RCTM_User, time=-0.0063, value=0.0184}
          {arriveTangent=0.3112, interpMode=RCIM_Cubic, leaveTangent=0.3112, tangentMode=RCTM_User, time=0.3734, value=0.0777}
          {arriveTangent=-1.0034, interpMode=RCIM_Cubic, leaveTangent=-1.0034, tangentMode=RCTM_User, time=0.6422, value=0.1616}
          {interpMode=RCIM_Cubic, tangentMode=RCTM_User, time=1.0011, value=0.002}
         ]}}}
      Scale Curve=&FN_FrontEndFloatModifier(Gameplay=0.5, Front End=0.25)
     )
    )) {isSetParameters=true}
  ]
  rendererClasses [&NiagaraSpriteRendererProperties]
  renderers [
   {rendererClass=&NiagaraSpriteRendererProperties}
   {rendererClass=&NiagaraSpriteRendererProperties, rendererIndex=1}
  ]
 }
 Light {
  emitterUpdateScript [
   &EmitterState(Life Cycle Mode=::NewEnumerator1, Loop Delay=0.1, Loop Duration=0.9, UseLoopDelay=true)
   &SpawnBurst_Instantaneous(Spawn Count=1, Use Spawn Probability=false)
  ]
  particleSpawnScript [
   &InitializeParticle(
    Lifetime=&UniformRangedFloat(Minimum=1, Maximum=1)
    Module.Write Color=true
    Module.Write Lifetime=true
    Module.Write Mass=true
    Module.Write Material Random=true
    Module.Write Position=false
    Module.Write Scale=false
    Module.Write SpriteRotation=false
    Module.Write SpriteSize=true
    Position=&SimulationPosition()
    Sprite Rotation=&UniformRangedFloat(Minimum=0, Maximum=360)
    Sprite Size=&Vector2DFromFloat(Value=&UniformRangedFloat(Minimum=1, Maximum=1))
   )
   &SphereLocation(MaskSpawnGroup=false, Module.Override Randomness=false, Offset=<0 -250 0>, Sphere Radius=150, Surface Only=false) {enabled=false}
  ]
  particleUpdateScript [
   &ParticleState()
   &SkeletalMeshLocation(
    Inherit Velocity=false
    MaskSpawnGroup=false
    Mesh Sampling Type=::NewEnumerator5
    Module.Override Randomness=false
    Skeletal Mesh=$User.SkeletalMesh
   )
   /SetVariables(Particles.GradientColor:Vector3f=&MakeVectorFromLinearColorRGB(Color=&MakeLinearColorFromVectorAndFloat(
      "Vector (RGB)"=&MakeVectorFromLinearColorRGB(Color=&ColorFromCurve(ColorCurve=DataInterface() {propertyValues {
          AlphaCurve {keys [<at -0.0023 1 RCIM_Cubic>,<at 0.9977 1 RCIM_Cubic>]}
          BlueCurve {keys [<at -0.0023 0 RCIM_Cubic>,<at 0.1706 0>,<at 0.4507 0>,<at 0.5951 1>,<at 0.7691 1.5>,<at 0.9071 1.5678 RCIM_Cubic>]}
          GreenCurve {keys [<at -0.0023 0.039 RCIM_Cubic>,<at 0.1706 0.8559>,<at 0.4507 1>,<at 0.5951 0.6207>,<at 0.7691 0.4532>,<at 0.9071 0 RCIM_Cubic>]}
          RedCurve {keys [<at -0.0023 1 RCIM_Cubic>,<at 0.1706 2>,<at 0.4507 0.3535>,<at 0.5951 0.0629>,<at 0.7691 0>,<at 0.9071 2 RCIM_Cubic>]}
         }}))
      "Float (Alpha)"=1
     ))) {isSetParameters=true, module=SetVariables_#4}
   /SetVariables(Particles.Color=&MakeLinearColorFromVectorAndFloat(
     "Vector (RGB)"=&Lerp_Vector(A=$Particles.GradientColor, B=&UniformRangedVector(Minimum=<15 15 15>, Maximum=<250 250 250>), Alpha=0)
     "Float (Alpha)"=&FloatFromCurve(
      FloatCurve=DataInterface() {propertyValues {Curve {keys [
          {interpMode=RCIM_Cubic, tangentMode=RCTM_User, time=-0.002, value=0.0241}
          {arriveTangent=-0.0221, interpMode=RCIM_Cubic, leaveTangent=-0.0221, tangentMode=RCTM_User, time=0.0954, value=0.9231}
          <at 1.0011 0.002>
         ]}}}
      Scale Curve=0.8
     )
    )) {enabled=false, isSetParameters=true}
   /SetVariables(Particles.Color=&MakeLinearColorFromVectorAndFloat(
     "Vector (RGB)"=&Lerp_Vector(
      A=&Multiply_VectorByFloat(Vector=$Particles.GradientColor, Float=10)
      B=&Multiply_VectorByFloat(Vector=$Particles.GradientColor, Float=140)
      Alpha=&FN_FrontEndFloatModifier(Gameplay=1, Front End=0)
     )
     "Float (Alpha)"=&FloatFromCurve(FloatCurve=DataInterface() {propertyValues {Curve {keys [
          {interpMode=RCIM_Cubic, tangentMode=RCTM_User, time=0, value=0}
          {arriveTangent=-0.8483, interpMode=RCIM_Cubic, leaveTangent=-0.8483, tangentMode=RCTM_User, time=0.5, value=1}
          {interpMode=RCIM_Cubic, tangentMode=RCTM_User, time=1, value=0}
         ]}}})
    )) {isSetParameters=true}
   &SolveForcesAndVelocity(Clamp Velocity=false, Limit Acceleration=false) {enabled=false}
   /Game/Effects/Niagara/Modules/Light/FN_Light_v5/FN_Light(Light Exponent=2, Light Radius=100, LightEnabled=true, Scale Color=200) {enabled=false}
   /SetVariables(Particles.LightPosition:NiagaraPosition=$Particles.Position) {enabled=false, isSetParameters=true}
   /SetVariables(Particles.LightExponent=1) {isSetParameters=true}
   /SetVariables() {isSetParameters=true}
   &SystemLocation(MaskSpawnGroup=false, Offset=<0 -50 150>, Offset Coordinate Space=::Simulation) {enabled=false}
  ]
  rendererClasses [&NiagaraLightRendererProperties]
  renderers [{rendererClass=&NiagaraLightRendererProperties, rendererIndex=0}]
 }
 Pop_Sparks {
  emitterUpdateScript [
   &EmitterState(Life Cycle Mode=::NewEnumerator1, Loop Duration=1, UseLoopDelay=false)
   &SpawnRate(SpawnRate=&Lerp_Float(A=0, B=100, Alpha=$System.SpawnRamp), Use Spawn Probability=true)
  ]
  particleSpawnScript [
   &InitializeParticle(
    Lifetime=&Multiply_Float(A=&UniformRangedFloat(Minimum=0.5, Maximum=0.7), B=1)
    Module.Write Color=true
    Module.Write Lifetime=true
    Module.Write Mass=true
    Module.Write Material Random=true
    Module.Write Position=false
    Module.Write Scale=false
    Module.Write SpriteRotation=false
    Module.Write SpriteSize=true
    Position=&SimulationPosition()
    Sprite Size=&UniformRangedVector2D(
     Minimum=&Vector2DFromFloat(Value=25)
     Maximum=&Vector2DFromFloat(Value=&FN_FrontEndFloatModifier(Gameplay=50, Front End=35))
    )
   )
   &SampleSkeletalMeshSkeleton(Mesh:NiagaraDataInterfaceSkeletalMesh=$User.SkeletalMesh)
   &SkeletalMeshSkeletonLocation(Inherit Velocity From Mesh=150, Module.Inherit Velocity=false, Module.Use Sampled Bone as Ribbon ID=true) {enabled=false}
   &SphereLocation(
    MaskSpawnGroup=false
    Module.Override Randomness=true
    Module.Override Seed=true
    Offset=&FN_FrontEndVector3Modifier(Gameplay=<0 0 0>, Front End=<0 0 0>)
    Randomness Mode:ENiagaraRandomnessMode=::NewEnumerator1
    Sphere Origin:Vector3f=$Particles.SampleSkeletalMeshSkeleton.SampledPosition
    Sphere Radius=&UniformRangedFloat(Minimum=30, Maximum=50)
    Surface Only=false
   )
   /SetVariables(Particles.Position_OffsetSphere:NiagaraPosition=$Particles.Position) {enabled=false, isSetParameters=true}
   &CurlNoiseForce(
    Noise Frequency=15
    Noise Strength=&Multiply_Float(A=0.15, B=&MakeFloatFromVector(Vector=$Engine.Owner.Velocity, Channel=::NewEnumerator0))
    Pan Noise=true
    Pan Noise Field=<0 10 0>
   ) {module=CurlNoiseForce001}
   &ApplyInitialForces()
   /SetVariables(Particles.RandomDrag:NiagaraFloat=&UniformRangedFloat(Minimum=1, Maximum=5)) {isSetParameters=true}
  ]
  particleUpdateScript [
   &ParticleState()
   /SetVariables(Particles.GradientColor:Vector3f=&MakeVectorFromLinearColorRGB(Color=/Niagara/DynamicInputs/TypeConversions/MakeLinearColorFromFloat(Value:NiagaraFloat=1))) {enabled=false, isSetParameters=true}
   &Color(
    Color=&ColorFromCurve(ColorCurve=DataInterface() {propertyValues {
       AlphaCurve {keys [<at -0.0023 1 RCIM_Cubic>,<at 0.9977 1 RCIM_Cubic>]}
       BlueCurve {keys [<at 0.1175 0 RCIM_Cubic>,<at 0.2682 0>,<at 0.4507 0>,<at 0.5951 1>,<at 0.7691 1.5>,<at 0.9071 1.5678 RCIM_Cubic>]}
       GreenCurve {keys [<at 0.1175 0.039 RCIM_Cubic>,<at 0.2682 0.8559>,<at 0.4507 1>,<at 0.5951 0.6207>,<at 0.7691 0.4532>,<at 0.9071 0 RCIM_Cubic>]}
       RedCurve {keys [<at 0.1175 1 RCIM_Cubic>,<at 0.2682 2>,<at 0.4507 0.3535>,<at 0.5951 0.0629>,<at 0.7691 0>,<at 0.9071 2 RCIM_Cubic>]}
      }})
    Scale Alpha=&FloatFromCurve(
     FloatCurve=DataInterface() {propertyValues {Curve {keys [
         {interpMode=RCIM_Cubic, tangentMode=RCTM_User, time=-0.0063, value=0.0184}
         {arriveTangent=-0.8483, interpMode=RCIM_Cubic, leaveTangent=-0.8483, tangentMode=RCTM_User, time=0.1718, value=1.7147}
         {arriveTangent=-5.0978, interpMode=RCIM_Cubic, leaveTangent=-5.0978, tangentMode=RCTM_User, time=0.2818, value=1.2143}
         {arriveTangent=-0.2783, interpMode=RCIM_Cubic, leaveTangent=-0.2783, tangentMode=RCTM_User, time=0.4294, value=1.1587}
         {arriveTangent=0.3112, interpMode=RCIM_Cubic, leaveTangent=0.3112, tangentMode=RCTM_User, time=0.5407, value=1.3419}
         {arriveTangent=-3.7119, interpMode=RCIM_Cubic, leaveTangent=-3.7119, tangentMode=RCTM_User, time=0.7242, value=0.3749}
         {interpMode=RCIM_Cubic, tangentMode=RCTM_User, time=1.0011, value=0.002}
        ]}}}
     Scale Curve=0.8
    )
    Scale Color=&VectorFromFloat(Value=15)
   )
   /SetVariables(Particles.Color=&MakeLinearColorFromVectorAndFloat(
     "Vector (RGB)"=&Lerp_Vector(
      A=&Multiply_VectorByFloat(Vector=$Particles.GradientColor, Float=5)
      B=&Multiply_VectorByFloat(Vector=$Particles.GradientColor, Float=25)
      Alpha=0.5
     )
     "Float (Alpha)"=&FloatFromCurve(
      FloatCurve=DataInterface() {propertyValues {Curve {keys [
          {interpMode=RCIM_Cubic, tangentMode=RCTM_User, time=-0.0063, value=0.0184}
          {arriveTangent=-0.8483, interpMode=RCIM_Cubic, leaveTangent=-0.8483, tangentMode=RCTM_User, time=0.1718, value=1.7147}
          {arriveTangent=-5.0978, interpMode=RCIM_Cubic, leaveTangent=-5.0978, tangentMode=RCTM_User, time=0.2818, value=1.2143}
          {arriveTangent=-0.2783, interpMode=RCIM_Cubic, leaveTangent=-0.2783, tangentMode=RCTM_User, time=0.4294, value=1.1587}
          {arriveTangent=0.3112, interpMode=RCIM_Cubic, leaveTangent=0.3112, tangentMode=RCTM_User, time=0.5407, value=1.3419}
          {arriveTangent=-3.7119, interpMode=RCIM_Cubic, leaveTangent=-3.7119, tangentMode=RCTM_User, time=0.7242, value=0.3749}
          {interpMode=RCIM_Cubic, tangentMode=RCTM_User, time=1.0011, value=0.002}
         ]}}}
      Scale Curve=0.8
     )
    )) {enabled=false, isSetParameters=true, module=SetVariables_#3}
   /SetVariables(Particles.Position=&Add_Vector(A=Linked() {linkedVariable {name=Particles.Position, type=NiagaraPosition}}, B=<0 0 0>)) {isSetParameters=true}
   &AddVelocity(Velocity=&RotateVectorByQuaternion(
     Vector To Rotate=&UniformRangedVector(Minimum=<-15 5 -15>, Maximum=<15 25 15>)
     Quaternion=$Particles.SampleSkeletalMeshSkeleton.SampledRotation
    )) {enabled=false}
   &AccelerationForce(Acceleration=<0 0 -100>) {enabled=false}
   &CurlNoiseForce(
    Noise Frequency=1
    Noise Quality / Cost=::NewEnumerator0
    Noise Strength=&Multiply_Float(A=0.25, B=&MakeFloatFromVector(Vector=$Engine.Owner.Velocity, Channel=::NewEnumerator0))
    Pan Noise=true
    Pan Noise Field=<0 10 0>
   )
   &Drag(Drag=$Particles.RandomDrag, Use Linear Drag=true, Use Rotational Drag=false)
   &InheritVelocity(
    Module.Limit Speed=true
    Velocity Scale=<0 -0.025 0>
    Velocity Source=&TransformVector(Vector=$Engine.Owner.Velocity, Source Space=::World, Destination Space=::Local)
   )
   &SolveForcesAndVelocity(Clamp Velocity=false, Limit Acceleration=false)
   &ScaleSpriteSize(Scale Factor=&Vector2DFromFloat(Value=&FloatFromCurve(FloatCurve=DataInterface() {propertyValues {Curve {keys [
          {arriveTangent=7.1136, interpMode=RCIM_Cubic, leaveTangent=7.1136, tangentMode=RCTM_User, time=0, value=0}
          {arriveTangent=1.4654, interpMode=RCIM_Cubic, leaveTangent=1.4654, time=0.18, value=1}
          <at 0.3814 0.5589>
          {arriveTangent=-0.0952, interpMode=RCIM_Cubic, leaveTangent=-0.0952, tangentMode=RCTM_User, time=0.7849, value=0.7663}
          <at 1 0.5 RCIM_Cubic>
         ]}}})))
   &ScaleSpriteSizeBySpeed(Max Scale Factor=<0.5 0.5>)
   &CameraOffset(Camera Offset Amount=25)
  ]
  rendererClasses [&NiagaraSpriteRendererProperties]
  renderers [{rendererClass=&NiagaraSpriteRendererProperties}]
 }
}
summary {
 systemName=NS_Glider_Donut_Optimized
 emitters [
  {emitterName=Pop_Sparks, rendererClasses [&NiagaraSpriteRendererProperties]}
  {emitterName=FN_Sprite, rendererClasses [&NiagaraSpriteRendererProperties]}
  {emitterName=Light, rendererClasses [&NiagaraLightRendererProperties]}
  {emitterName=Confetti, rendererClasses [&NiagaraSpriteRendererProperties]}
  {emitterName=FN_Contrails_SpriteCloudy, rendererClasses [&NiagaraSpriteRendererProperties]}
 ]
 userVariables [
  User.bIsFrontEnd {type=NiagaraBool}
  User.bIsFrontEndPreview {type=NiagaraBool}
  User.NewPosition {type=Vector3f, description="", defaultValue {struct=Vector3f, value=<0 0 0>}}
  User.PlayerSpeed {type=NiagaraFloat, defaultValue=0, description=""}
  User.RibbonIDOffset {type=NiagaraInt32}
  User.SkeletalMesh {
   type=NiagaraDataInterfaceSkeletalMesh
   defaultValue {
    struct=NiagaraExt_VariableValue_DataInterface
    value {dataInterface=&self:NiagaraDataInterfaceSkeletalMesh_0}
   }
  }
  User.Sparks {type=NiagaraFloat, defaultValue=1, description=""}
 ]
}
